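# Physics asset for the SteamMarauder backpack cosmetic, defining collision bodies and joints for its dangling metal pieces.
# Locked angular constraints link root, pack, backpack and the two metal base bones.
# From those bases, a four-link lower-right chain and a single upper-left link move within tight swing and twist limits.
# An unconstrained pelvis body adds a box and small capsule, apparently as a body collider for the pieces.
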PhysicsAsset /BRCosmetics/Accessories/FORT_Backpacks/Backpack_F_MED_SteamMarauder/Meshes/F_MED_SteamMarauder_Pack_Physics
bodies {
 dyn_BACE_mettal_down_R_0 [{shapeType=Capsule, length=0.505, radius=0.505, shapeName=dyn_BACE_mettal_down_R_0_capsule, rotation=<rot 0 -90 0>}]
 dyn_BACE_mettal_up_L_0 [{shapeType=Capsule, length=0.505, radius=0.505, shapeName=dyn_BACE_mettal_up_R_0_capsule, rotation=<rot 0 -90 0>}]
 dyn_backpack [{shapeType=Capsule, length=21.8506, radius=0.9236, center=<-0.0417 -0.0968 -11.61>, rotation=<rot 0.2051 -90 0.1869>}]
 dyn_mettal_down_R_1 [{shapeType=Capsule, length=1, radius=1, center=<1.9626 0 0>}]
 dyn_mettal_down_R_2 [{
   shapeType=Capsule
   length=4.1398
   radius=1.6119
   shapeName=dyn_mettal_down_R_2_capsule
   center=<1.3744 -0.3093 -0.0073>
   rotation=<rot 0 81.3699 -89.7361>
  }]
 dyn_mettal_down_R_3 [{
   shapeType=Capsule
   length=5.1279
   radius=1.0379
   shapeName=dyn_mettal_down_R_3_capsule
   center=<4.4193 0.3858 0.0653>
   rotation=<rot 0 96.4613 -89.3463>
  }]
 dyn_mettal_down_R_4 [{
   shapeType=Capsule
   length=6.4939
   radius=2.938
   shapeName=dyn_mettal_down_R_4_capsule
   center=<5.0672 1.4413 -0.3116>
   rotation=<rot 0 -85.6398 -87.4645>
  }]
 dyn_mettal_up_L_1 [{shapeType=Capsule, length=1, radius=1}]
 pack [{shapeType=Capsule, length=0.505, radius=0.505, shapeName=pack_capsule, rotation=<rot 0 -90 0>}]
 pelvis [
  {shapeType=Capsule, length=0.505, radius=0.505, shapeName=pelvis_capsule, center=<0 0 -12.0886>, rotation=<rot 0 -90 0>}
  {shapeType=Box, extentX=27.7, extentY=7.7, extentZ=22.1, shapeName=pelvis_box, center=<0 0 -11.9476>, rotation=<rot 0 0 18.6143>}
 ]
 root [{shapeType=Capsule, length=0.505, radius=0.505, shapeName=root_capsule, rotation=<rot 0 -90 0>}]
}
constraints [
 {
  bone1Name=dyn_BACE_mettal_down_R_0
  bone2Name=dyn_backpack
  swing1LimitDegrees=45
  swing1Motion=Locked
  swing2LimitDegrees=45
  swing2Motion=Locked
  twistLimitDegrees=45
  twistMotion=Locked
 }
 {
  bone1Name=dyn_backpack
  bone2Name=pack
  swing1LimitDegrees=45
  swing1Motion=Locked
  swing2LimitDegrees=45
  swing2Motion=Locked
  twistLimitDegrees=45
  twistMotion=Locked
 }
 {
  bone1Name=pack
  bone2Name=root
  swing1LimitDegrees=45
  swing1Motion=Locked
  swing2LimitDegrees=45
  swing2Motion=Locked
  twistLimitDegrees=45
  twistMotion=Locked
 }
 {bone1Name=dyn_mettal_down_R_1, bone2Name=dyn_BACE_mettal_down_R_0, swing1LimitDegrees=18, swing2LimitDegrees=12, twistLimitDegrees=4}
 {bone1Name=dyn_mettal_down_R_2, bone2Name=dyn_mettal_down_R_1, swing1LimitDegrees=6, swing2LimitDegrees=6, twistLimitDegrees=5}
 {bone1Name=dyn_mettal_down_R_3, bone2Name=dyn_mettal_down_R_2, swing1LimitDegrees=3, swing2LimitDegrees=3, twistLimitDegrees=1}
 {bone1Name=dyn_mettal_down_R_4, bone2Name=dyn_mettal_down_R_3, swing1LimitDegrees=6, swing2LimitDegrees=6, twistLimitDegrees=2}
 {bone1Name=dyn_mettal_up_L_1, bone2Name=dyn_BACE_mettal_up_L_0, swing1LimitDegrees=15, swing2LimitDegrees=30, twistLimitDegrees=5}
 {
  bone1Name=dyn_BACE_mettal_up_L_0
  bone2Name=dyn_backpack
  swing1LimitDegrees=45
  swing1Motion=Locked
  swing2LimitDegrees=45
  swing2Motion=Locked
  twistLimitDegrees=45
  twistMotion=Locked
 }
]
